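# Widget Blueprint for one voice chat speaker entry, showing avatar, display name, channel name and an active-voice icon, on native FortVoiceChannelSpeakerWidget.
# Anim_Open and Anim_Close animate the bounds, content and green edge indicator; this graph never plays them.
# OnAnimateTalkingChanged drives the icon material's AnimateStart through AnimateVoice, reversing it if mid-play, otherwise playing forward or backward.
# OnStyleAssigned switches on style, but every branch is empty.
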WidgetBlueprint /Game/UI/Social/FortVoiceChannelSpeaker
animations [
 Anim_Open {
  binding_count=3
  length_seconds=0.16
  object=&self:Anim_Open
  sequence {
   bindings [
    Content {
     id=#1
     possessed_class=HorizontalBox
     tracks [
      {
       type=MovieSceneFloatTrack
       display_name=Render Opacity
       sections [{type=MovieSceneFloatSection, channels [None {type=MovieSceneScriptingFloatChannel, keys [[8,0],[16,1]]}]}]
      }
      {
       type=MovieScene2DTransformTrack
       sections [{
         type=MovieScene2DTransformSection
         channels [
          Translation.X {type=MovieSceneScriptingFloatChannel, keys [[8,-32],[16,0]]}
          Translation.Y {type=MovieSceneScriptingFloatChannel}
          Angle {type=MovieSceneScriptingFloatChannel}
          Scale.X {type=MovieSceneScriptingFloatChannel}
          Scale.Y {type=MovieSceneScriptingFloatChannel}
          Shear.X {type=MovieSceneScriptingFloatChannel}
          Shear.Y {type=MovieSceneScriptingFloatChannel}
         ]
        }]
      }
     ]
    }
    Border_OnlineStatusIndicator {
     id=#2
     possessed_class=Border
     tracks [
      {
       type=MovieScene2DTransformTrack
       sections [{
         type=MovieScene2DTransformSection
         channels [
          Translation.X {type=MovieSceneScriptingFloatChannel}
          Translation.Y {type=MovieSceneScriptingFloatChannel}
          Angle {type=MovieSceneScriptingFloatChannel}
          Scale.X {type=MovieSceneScriptingFloatChannel, keys [[0,1],[8,1.5],[16,1]]}
          Scale.Y {type=MovieSceneScriptingFloatChannel, keys [[0,1],[8,1.25],[16,1]]}
          Shear.X {type=MovieSceneScriptingFloatChannel}
          Shear.Y {type=MovieSceneScriptingFloatChannel}
         ]
        }]
      }
      {
       type=MovieSceneFloatVectorTrack
       display_name=Desired Size Scale
       sections [{
         type=MovieSceneFloatVectorSection
         channels [
          Vector.X {type=MovieSceneScriptingFloatChannel, keys [[0,0],[4,1]]}
          Vector.Y {type=MovieSceneScriptingFloatChannel}
         ]
        }]
      }
      {
       type=MovieSceneFloatTrack
       display_name=Render Opacity
       sections [{type=MovieSceneFloatSection, channels [None {type=MovieSceneScriptingFloatChannel, keys [[0,0],[4,1]]}]}]
      }
     ]
    }
    Border_ScaleBound {
     id=#3
     possessed_class=CommonBorder
     tracks [
      {
       type=MovieSceneFloatVectorTrack
       display_name=Desired Size Scale
       sections [{
         type=MovieSceneFloatVectorSection
         channels [
          Vector.X {type=MovieSceneScriptingFloatChannel, keys [[0,0],[16,1]]}
          Vector.Y {type=MovieSceneScriptingFloatChannel}
         ]
        }]
      }
      {
       type=MovieSceneFloatTrack
       display_name=Render Opacity
       sections [{type=MovieSceneFloatSection, channels [None {type=MovieSceneScriptingFloatChannel, keys [[0,0],[4,1]]}]}]
      }
     ]
    }
   ]
   display_rate {numerator=100}
  }
 }
 Anim_Close {
  binding_count=3
  length_seconds=0.16
  object=&self:Anim_Close
  sequence {
   bindings [
    Border_OnlineStatusIndicator {
     id=#4
     possessed_class=Border
     tracks [
      {
       type=MovieScene2DTransformTrack
       sections [{
         type=MovieScene2DTransformSection
         channels [
          Translation.X {type=MovieSceneScriptingFloatChannel}
          Translation.Y {type=MovieSceneScriptingFloatChannel}
          Angle {type=MovieSceneScriptingFloatChannel}
          Scale.X {type=MovieSceneScriptingFloatChannel, keys [[0,1],[8,1.5],[16,1]]}
          Scale.Y {type=MovieSceneScriptingFloatChannel, keys [[0,1],[8,1.25],[16,1]]}
          Shear.X {type=MovieSceneScriptingFloatChannel}
          Shear.Y {type=MovieSceneScriptingFloatChannel}
         ]
        }]
      }
      {
       type=MovieSceneFloatVectorTrack
       display_name=Desired Size Scale
       sections [{
         type=MovieSceneFloatVectorSection
         channels [
          Vector.X {type=MovieSceneScriptingFloatChannel, keys [[12,1],[16,0]]}
          Vector.Y {type=MovieSceneScriptingFloatChannel}
         ]
        }]
      }
      {
       type=MovieSceneFloatTrack
       display_name=Render Opacity
       sections [{type=MovieSceneFloatSection, channels [None {type=MovieSceneScriptingFloatChannel, keys [[12,1],[16,0]]}]}]
      }
     ]
    }
    Content {
     id=#5
     possessed_class=HorizontalBox
     tracks [
      {
       type=MovieSceneFloatTrack
       display_name=Render Opacity
       sections [{type=MovieSceneFloatSection, channels [None {type=MovieSceneScriptingFloatChannel, keys [[0,1],[8,0]]}]}]
      }
      {
       type=MovieScene2DTransformTrack
       sections [{
         type=MovieScene2DTransformSection
         channels [
          Translation.X {type=MovieSceneScriptingFloatChannel, keys [[0,0],[8,-32]]}
          Translation.Y {type=MovieSceneScriptingFloatChannel}
          Angle {type=MovieSceneScriptingFloatChannel}
          Scale.X {type=MovieSceneScriptingFloatChannel}
          Scale.Y {type=MovieSceneScriptingFloatChannel}
          Shear.X {type=MovieSceneScriptingFloatChannel}
          Shear.Y {type=MovieSceneScriptingFloatChannel}
         ]
        }]
      }
     ]
    }
    Border_ScaleBound {
     id=#6
     possessed_class=CommonBorder
     tracks [
      {
       type=MovieSceneFloatVectorTrack
       display_name=Desired Size Scale
       sections [{
         type=MovieSceneFloatVectorSection
         channels [
          Vector.X {type=MovieSceneScriptingFloatChannel, keys [[0,1],[16,0]]}
          Vector.Y {type=MovieSceneScriptingFloatChannel}
         ]
        }]
      }
      {
       type=MovieSceneFloatTrack
       display_name=Render Opacity
       sections [{type=MovieSceneFloatSection, channels [None {type=MovieSceneScriptingFloatChannel, keys [[12,1],[16,0]]}]}]
      }
     ]
    }
   ]
   display_rate {numerator=100}
  }
 }
 AnimateVoice {
  binding_count=1
  length_seconds=0.25
  object=&self:AnimateVoice
  sequence {bindings [Image_ActiveVoiceIcon {
     id=#7
     possessed_class=Image
     tracks [{
       type=MovieSceneWidgetMaterialTrack
       sections [{type=MovieSceneParameterSection, channels [AnimateStart {type=MovieSceneScriptingFloatChannel, keys [[0,0],[5,1]]}]}]
      }]
    }]}
 }
]
graphs [EventGraph {nodes [
   n1 = Event OnStyleAssigned!() -> (OutputDelegate, then, NewStyle:EChannelSpeakerStyle Enum=InGame) {is=title K2Node_Event}
   > Switch on EChannelSpeakerStyle!(execute, Selection:EChannelSpeakerStyle Enum=@n1.NewStyle) -> (InGame:Exec, InLobby:Exec, OutOfClient:Exec, Max:Exec) {is=title K2Node_SwitchEnum}
   n2 = Event OnAnimateTalkingChanged!() -> (OutputDelegate, then, bIsNowTalking:bool) {is=title K2Node_Event}
   > n3 = Branch(Condition=@n2.bIsNowTalking)
   n4 = IsAnimationPlaying(InAnimation=Get AnimateVoice() -> (AnimateVoice:Widget Animation Object Reference))
   > n5 = Branch(Condition=@n4)
   n6 = Get AnimateVoice()
   > ReverseAnimation!(execute, self:User Widget Object Reference, InAnimation:Widget Animation Object Reference=@n6) -> (then)
   PlayAnimation(execute=@n5.else, InAnimation=@n6)
   n7 = Branch(execute=@n3.else, Condition=@n4)
   n8 = Get AnimateVoice()
   > ReverseAnimation(InAnimation=@n8)
   PlayAnimation(execute=@n7.else, InAnimation=@n8, PlayMode=Reverse)
   Event PreConstruct()
  ]}]
variables []
widgets {
 info {inheritedWidgetCount=5, parentClass=/Script/FortniteUI.FortVoiceChannelSpeakerWidget, rootWidgetClass=&Overlay, widgetCount=13}
 widgets [
  Overlay_0 {widgetClassPath=&Overlay, widget=Overlay_0}
  Border_ScaleBound {
   widgetClassPath=&CommonBorder
   bInherited=true
   bIsVariable=true
   parent=Overlay_0
   widget=Border_ScaleBound
   properties {
    clipping=ClipToBounds
    horizontalAlignment=HAlign_Center
    verticalAlignment=VAlign_Center
    background {
     bIsDynamicallyLoaded=false
     drawAs=Image
     imageType=NoImage
     mirroring=NoMirror
     resourceName=None
     resourceObject=/Game/UI/SocialPanel/Art/Materials/M_UI_SocialUserItemBorder_UserListItem
     tiling=NoTile
     imageSize=<32 32>
     margin=<ltrb 0 0 0 0>
     outlineSettings {
      bUseBrushTransparency=false
      roundingType=HalfHeightRadius
      width=0
      color {colorUseRule=UseColor_Specified, specifiedColor=<rgba 0 0 0 0>}
      cornerRadii=<0 0 0 0>
     }
     tintColor {colorUseRule=UseColor_Specified, specifiedColor=<rgba 1 1 1 1>}
     uVRegion {bIsValid=false, max=<0 0>, min=<0 0>}
    }
    padding=<ltrb 0 0 0 0>
   }
   slot {class=OverlaySlot, properties {verticalAlignment=VAlign_Center}}
  }
  SizeBox_1 {
   widgetClassPath=&SizeBox
   parent=Border_ScaleBound
   widget=SizeBox_1
   properties {bOverride_MinDesiredHeight=true, bOverride_MinDesiredWidth=true, heightOverride=48, minDesiredHeight=48, minDesiredWidth=192}
   slot {class=BorderSlot, properties {horizontalAlignment=HAlign_Center, verticalAlignment=VAlign_Center, padding=<ltrb 0 0 0 0>}}
  }
  Content {
   widgetClassPath=&HorizontalBox
   bIsVariable=true
   parent=SizeBox_1
   widget=Content
   slot {class=SizeBoxSlot, properties {horizontalAlignment=HAlign_Left, verticalAlignment=VAlign_Center, padding=<ltrb 8 0 4 0>}}
  }
  SizeBox_0 {
   widgetClassPath=&SizeBox
   parent=Content
   widget=SizeBox_0
   properties {
    bOverride_MinDesiredHeight=true
    bOverride_MinDesiredWidth=true
    heightOverride=48
    minDesiredHeight=48
    minDesiredWidth=48
    widthOverride=48
   }
   slot {class=HorizontalBoxSlot}
  }
  ScalesForMobile {widgetClassPath=&ScaleBox, parent=SizeBox_0, widget=ScalesForMobile, properties {stretch=ScaleToFitY}, slot {class=SizeBoxSlot}}
  AvatarIcon {
   widgetClassPath=/Script/FortniteUI.FortSocialAvatarIcon
   bInherited=true
   bIsVariable=true
   parent=ScalesForMobile
   widget=AvatarIcon
   properties {
    avatarMaterial=/Game/UI/SocialPanel/Art/Materials/UI_UserPortrait_Face
    materialTextureParamName=SmallPreview
    brush {
     bIsDynamicallyLoaded=false
     drawAs=Image
     imageType=NoImage
     mirroring=NoMirror
     resourceName=None
     resourceObject=&UI_UserPortrait_Face
     tiling=NoTile
     imageSize=<32 32>
     margin=<ltrb 0 0 0 0>
     outlineSettings {
      bUseBrushTransparency=false
      roundingType=HalfHeightRadius
      width=0
      color {colorUseRule=UseColor_Specified, specifiedColor=<rgba 0 0 0 0>}
      cornerRadii=<0 0 0 0>
     }
     tintColor {colorUseRule=UseColor_Specified, specifiedColor=<rgba 1 1 1 1>}
     uVRegion {bIsValid=false, max=<0 0>, min=<0 0>}
    }
   }
   slot {class=ScaleBoxSlot}
  }
  VerticalBox_0 {
   widgetClassPath=&VerticalBox
   parent=Content
   widget=VerticalBox_0
   slot {class=HorizontalBoxSlot, properties {verticalAlignment=VAlign_Center, padding=<ltrb 12 2 8 2>}}
  }
  Text_DisplayName {
   widgetClassPath=&CommonTextBlock
   bInherited=true
   parent=VerticalBox_0
   widget=Text_DisplayName
   properties {
    mobileFontSizeMultiplier=1.2
    text=USER NAME
    colorAndOpacity {colorUseRule=UseColor_Specified, specifiedColor=<rgba 0.4452 1 1 1>}
    font {
     bForceMonospaced=false
     bMaterialIsStencil=false
     fontMaterial=null
     fontObject=&BurbankSmall
     letterSpacing=0
     monospacedWidth=1
     size=15
     skewAmount=0
     typefaceFontName=Black
     outlineSettings {
      bApplyOutlineToDropShadows=false
      bMiteredCorners=false
      bSeparateFillAlpha=false
      outlineMaterial=null
      outlineSize=0
      outlineColor=<rgba 0 0 0 1>
     }
    }
    shadowOffset=<0 0>
   }
   slot {class=VerticalBoxSlot, properties {verticalAlignment=VAlign_Top, size {sizeRule=Fill, value=1}}}
  }
  Text_ChannelName {
   widgetClassPath=&CommonTextBlock
   bInherited=true
   parent=VerticalBox_0
   widget=Text_ChannelName
   properties {
    mobileFontSizeMultiplier=1.2
    text=CHANNEL NAME
    textTransformPolicy=ToUpper
    colorAndOpacity {colorUseRule=UseColor_Specified, specifiedColor=<rgba 0.4452 1 1 1>}
    font {
     bForceMonospaced=false
     bMaterialIsStencil=false
     fontMaterial=null
     fontObject=&BurbankSmall
     letterSpacing=0
     monospacedWidth=1
     size=12
     skewAmount=0
     typefaceFontName=Black
     outlineSettings {
      bApplyOutlineToDropShadows=false
      bMiteredCorners=false
      bSeparateFillAlpha=false
      outlineMaterial=null
      outlineSize=0
      outlineColor=<rgba 0 0 0 1>
     }
    }
    shadowOffset=<0 0>
   }
   slot {class=VerticalBoxSlot, properties {verticalAlignment=VAlign_Bottom, size {sizeRule=Fill, value=1}}}
  }
  Image_ActiveVoiceIcon {
   widgetClassPath=&Image
   bInherited=true
   bIsVariable=true
   parent=Content
   widget=Image_ActiveVoiceIcon
   properties {
    visibility=SelfHitTestInvisible
    brush {
     bIsDynamicallyLoaded=false
     drawAs=Image
     imageType=NoImage
     mirroring=NoMirror
     resourceName=None
     resourceObject=/Game/UI/SocialPanel/Art/Materials/M_Icon_ActiveVoice_Inst
     tiling=NoTile
     imageSize=<48 32>
     margin=<ltrb 0 0 0 0>
     outlineSettings {
      bUseBrushTransparency=false
      roundingType=HalfHeightRadius
      width=0
      color {colorUseRule=UseColor_Specified, specifiedColor=<rgba 0 0 0 0>}
      cornerRadii=<0 0 0 0>
     }
     tintColor {colorUseRule=UseColor_Specified, specifiedColor=<rgba 1 1 1 1>}
     uVRegion {bIsValid=false, max=<0 0>, min=<0 0>}
    }
   }
   slot {class=HorizontalBoxSlot, properties {horizontalAlignment=HAlign_Center, verticalAlignment=VAlign_Center}}
  }
  Border_OnlineStatusIndicator {
   widgetClassPath=&Border
   bIsVariable=true
   parent=Overlay_0
   widget=Border_OnlineStatusIndicator
   properties {visibility=SelfHitTestInvisible, brushColor=<rgba 0.1946 0.9823 0.1812 1>, padding=<ltrb 0 0 0 0>}
   slot {class=OverlaySlot, properties {verticalAlignment=VAlign_Fill}}
  }
  Spacer_0 {
   widgetClassPath=&Spacer
   parent=Border_OnlineStatusIndicator
   widget=Spacer_0
   properties {size=<8 8>}
   slot {class=BorderSlot, properties {padding=<ltrb 0 0 0 0>}}
  }
 ]
}
