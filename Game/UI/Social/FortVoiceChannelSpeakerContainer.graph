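# Voice-chat widget for the social UI that shows one channel's speakers under a labeled header, built on FortVoiceSpeakerContainerWidget.
# The header pairs a tinted speaker-status material icon with a channel name text, defaulting to PARTY, both scaled up for mobile.
# Below it, a size box reserves height for a dynamic entry box that spawns FortVoiceChannelSpeaker entries.
# The event graph is empty, leaving behavior to the native parent.
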WidgetBlueprint /Game/UI/Social/FortVoiceChannelSpeakerContainer
graphs [EventGraph {links [], nodes []}]
variables []
widgets {
 info {
  inheritedWidgetCount=3
  parentClass=/Script/FortniteUI.FortVoiceSpeakerContainerWidget
  rootWidgetClass=&VerticalBox
  widgetCount=6
 }
 widgets [
  VerticalBox_0 {widgetClassPath=&VerticalBox, widget=VerticalBox_0}
  Panel_ChannelContainer {
   widgetClassPath=&HorizontalBox
   bInherited=true
   bIsVariable=true
   parent=VerticalBox_0
   widget=Panel_ChannelContainer
   slot {class=VerticalBoxSlot}
  }
  SpeakerIcon {
   widgetClassPath=/Script/FortniteUI.FortMobileImage
   bIsVariable=true
   parent=Panel_ChannelContainer
   widget=SpeakerIcon
   properties {
    mobileSizeMultiplier=1.4
    visibility=SelfHitTestInvisible
    brush {
     bIsDynamicallyLoaded=false
     drawAs=Image
     imageType=NoImage
     mirroring=NoMirror
     resourceName=None
     resourceObject=/Game/Athena/HUD/Art/M_VoiceChatStatusIcon
     tiling=NoTile
     imageSize=<32 32>
     margin=<ltrb 0 0 0 0>
     outlineSettings {
      bUseBrushTransparency=false
      roundingType=HalfHeightRadius
      width=0
      color {colorUseRule=UseColor_Specified, specifiedColor=<rgba 0 0 0 0>}
      cornerRadii=<0 0 0 0>
     }
     tintColor {colorUseRule=UseColor_Specified, specifiedColor=<rgba 1 1 1 0.5>}
     uVRegion {bIsValid=false, max=<0 0>, min=<0 0>}
    }
    colorAndOpacity=<rgba 0.4452 1 1 1>
   }
   slot {class=HorizontalBoxSlot, properties {padding=<ltrb 8 0 0 0>}}
  }
  Text_ChannelDisplayName {
   widgetClassPath=&CommonTextBlock
   bInherited=true
   parent=Panel_ChannelContainer
   widget=Text_ChannelDisplayName
   properties {
    mobileFontSizeMultiplier=1.4
    text=PARTY
    colorAndOpacity {colorUseRule=UseColor_Specified, specifiedColor=<rgba 0.4452 1 1 0.5>}
    font {
     bForceMonospaced=false
     bMaterialIsStencil=false
     fontMaterial=null
     fontObject=&BurbankSmall
     letterSpacing=0
     monospacedWidth=1
     size=17
     skewAmount=0
     typefaceFontName=Black
     outlineSettings {
      bApplyOutlineToDropShadows=false
      bMiteredCorners=false
      bSeparateFillAlpha=false
      outlineMaterial=null
      outlineSize=0
      outlineColor=<rgba 0 0 0 1>
     }
    }
    shadowOffset=<0 0>
   }
   slot {class=HorizontalBoxSlot, properties {verticalAlignment=VAlign_Center}}
  }
  SizeBox_HeightReservation {
   widgetClassPath=&SizeBox
   parent=VerticalBox_0
   widget=SizeBox_HeightReservation
   properties {bOverride_MinDesiredHeight=true, heightOverride=32, minDesiredHeight=52}
   slot {class=VerticalBoxSlot}
  }
  DynamicEntry_Speakers {
   widgetClassPath=&DynamicEntryBox
   bInherited=true
   bIsVariable=true
   parent=SizeBox_HeightReservation
   widget=DynamicEntry_Speakers
   properties {entryWidgetClass=&/FortVoiceChannelSpeaker.FortVoiceChannelSpeaker_C, entrySpacing=<10 0>}
   slot {class=SizeBoxSlot, properties {horizontalAlignment=HAlign_Left, verticalAlignment=VAlign_Top}}
  }
 ]
}
